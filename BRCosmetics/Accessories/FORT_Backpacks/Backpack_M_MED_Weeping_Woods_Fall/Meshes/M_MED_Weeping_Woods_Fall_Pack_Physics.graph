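# Physics asset for the Weeping Woods Fall backpack cosmetic, covering its hive and dangling honey bones.
# Small capsule bodies branch from Dyn_Master_HIve into four strands: two single links (Master_Honey1, Honey1), a two-link chain (Honey2, Honey3), and a three-link chain (Master_Honey2, Honey, Honey4).
# The roots of both multi-link chains rotate freely on swing and twist, while the other joints carry moderate swing limits and tight twist limits.
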PhysicsAsset /BRCosmetics/Accessories/FORT_Backpacks/Backpack_M_MED_Weeping_Woods_Fall/Meshes/M_MED_Weeping_Woods_Fall_Pack_Physics
bodies {
 Dyn_Honey [{shapeType=Capsule, length=0.5, radius=1, center=<0.6011 0 0>, rotation=<rot 90 0 0>}]
 Dyn_Honey1 [{shapeType=Capsule, length=1.5, radius=1, center=<1.322 0 0>, rotation=<rot -90 0 0>}]
 Dyn_Honey2 [{shapeType=Capsule, length=2, radius=1, center=<1 0 0>, rotation=<rot 90 0 0>}]
 Dyn_Honey3 [{shapeType=Capsule, length=1, radius=1, center=<1.7327 0.0001 0>, rotation=<rot -90 0 0>}]
 Dyn_Honey4 [{shapeType=Capsule, length=0.505, radius=1, center=<0.6442 0 0>, rotation=<rot 90 0 0>}]
 Dyn_Master_HIve [{shapeType=Capsule, length=1, radius=1}]
 Dyn_Master_Honey1 [{shapeType=Capsule, length=1.5, radius=1.2, center=<0.05 0 0>}]
 Dyn_Master_Honey2 [{shapeType=Capsule, length=0.5, radius=0.5}]
}
constraints [
 {bone1Name=Dyn_Honey, bone2Name=Dyn_Master_Honey2, swing1LimitDegrees=40, swing2LimitDegrees=40, twistLimitDegrees=5}
 {bone1Name=Dyn_Honey3, bone2Name=Dyn_Honey2, swing1LimitDegrees=45, swing2LimitDegrees=45, twistLimitDegrees=5}
 {bone1Name=Dyn_Master_Honey1, bone2Name=Dyn_Master_HIve, swing1LimitDegrees=45, swing2LimitDegrees=45, twistLimitDegrees=5}
 {
  bone1Name=Dyn_Master_Honey2
  bone2Name=Dyn_Master_HIve
  swing1LimitDegrees=45
  swing1Motion=Free
  swing2LimitDegrees=45
  swing2Motion=Free
  twistLimitDegrees=45
  twistMotion=Free
 }
 {bone1Name=Dyn_Honey1, bone2Name=Dyn_Master_HIve, swing1LimitDegrees=45, swing2LimitDegrees=45, twistLimitDegrees=0.5}
 {
  bone1Name=Dyn_Honey2
  bone2Name=Dyn_Master_HIve
  swing1LimitDegrees=45
  swing1Motion=Free
  swing2LimitDegrees=45
  swing2Motion=Free
  twistLimitDegrees=45
  twistMotion=Free
 }
 {bone1Name=Dyn_Honey4, bone2Name=Dyn_Honey, swing1LimitDegrees=40, swing2LimitDegrees=40, twistLimitDegrees=5}
]
