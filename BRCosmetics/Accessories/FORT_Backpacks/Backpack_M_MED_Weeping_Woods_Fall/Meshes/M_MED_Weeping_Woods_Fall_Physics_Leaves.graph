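# Physics asset for the leaves on the Weeping Woods Fall backpack cosmetic, defining collision bodies and joint limits for their physical motion.
# One large capsule on the hive master bone serves as the hub, with twelve small leaf capsules each constrained directly to it rather than chained.
# Leaves share the same angular twist limit, while a few have tighter swing limits than the rest.
# Leaf numbering skips 6.
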PhysicsAsset /BRCosmetics/Accessories/FORT_Backpacks/Backpack_M_MED_Weeping_Woods_Fall/Meshes/M_MED_Weeping_Woods_Fall_Physics_Leaves
bodies {
 Dyn_Master_HIve [{shapeType=Capsule, length=10.5, radius=11, center=<1.5125 -1.1385 0.2096>, rotation=<rot 0 0 90>}]
 Dyn_leaf1 [{shapeType=Capsule, length=2.468, radius=0.8134, center=<2.1374 0.506 0.4216>, rotation=<rot 0 106.9481 -79.8563>}]
 Dyn_leaf10 [{shapeType=Capsule, length=2.723, radius=1.0458, center=<2.4167 0.6696 -0.0101>, rotation=<rot 0 113.1415 -86.7559>}]
 Dyn_leaf11 [{shapeType=Capsule, length=3.2858, radius=1.3565, center=<3.1563 0.3568 -0.0912>, rotation=<rot 0 -80.9013 -88.9256>}]
 Dyn_leaf12 [{shapeType=Capsule, length=3.2513, radius=1.2699, center=<2.8508 1.0155 -0.0292>, rotation=<rot 0 113.0446 -89.5849>}]
 Dyn_leaf13 [{shapeType=Capsule, length=1.5919, radius=0.6898, center=<1.4487 0.3869 0.052>, rotation=<rot 0 107.952 -89.7842>}]
 Dyn_leaf2 [{shapeType=Capsule, length=3.7158, radius=1.5009, center=<3.2661 1.0659 0.0742>, rotation=<rot 0 112.1731 -85.7109>}]
 Dyn_leaf3 [{shapeType=Capsule, length=1.6904, radius=0.7333, center=<1.4018 0.1757 -0.1347>, rotation=<rot 0 -79.0816 -74.2277>}]
 Dyn_leaf4 [{shapeType=Capsule, length=2.7492, radius=1.0957, center=<2.4808 0.4892 -0.0498>, rotation=<rot 0 -75.808 -84.8082>}]
 Dyn_leaf5 [{shapeType=Capsule, length=3.3902, radius=1.2461, center=<3.2167 -0.1303 0.1913>, rotation=<rot 0 86.4594 -88.5289>}]
 Dyn_leaf7 [{shapeType=Capsule, length=2.4424, radius=0.9636, center=<2.1931 0.0239 0.0604>, rotation=<rot 0 -90.3077 -85.8242>}]
 Dyn_leaf8 [{shapeType=Capsule, length=1.9536, radius=0.567, center=<1.6135 -0.1836 0.0866>, rotation=<rot 0 87.1983 -87.9796>}]
 Dyn_leaf9 [{shapeType=Capsule, length=2.3122, radius=0.8235, center=<2.042 0.3253 -0.0009>, rotation=<rot 0 -80.4845 -85.1413>}]
}
constraints [
 {bone1Name=Dyn_leaf1, bone2Name=Dyn_Master_HIve, swing1LimitDegrees=45, swing2LimitDegrees=45, twistLimitDegrees=45}
 {bone1Name=Dyn_leaf2, bone2Name=Dyn_Master_HIve, swing1LimitDegrees=45, swing2LimitDegrees=45, twistLimitDegrees=45}
 {bone1Name=Dyn_leaf3, bone2Name=Dyn_Master_HIve, swing1LimitDegrees=20, swing2LimitDegrees=20, twistLimitDegrees=45}
 {bone1Name=Dyn_leaf4, bone2Name=Dyn_Master_HIve, swing1LimitDegrees=45, swing2LimitDegrees=45, twistLimitDegrees=45}
 {bone1Name=Dyn_leaf5, bone2Name=Dyn_Master_HIve, swing1LimitDegrees=45, swing2LimitDegrees=45, twistLimitDegrees=45}
 {bone1Name=Dyn_leaf7, bone2Name=Dyn_Master_HIve, swing1LimitDegrees=45, swing2LimitDegrees=45, twistLimitDegrees=45}
 {bone1Name=Dyn_leaf8, bone2Name=Dyn_Master_HIve, swing1LimitDegrees=20, swing2LimitDegrees=20, twistLimitDegrees=45}
 {bone1Name=Dyn_leaf9, bone2Name=Dyn_Master_HIve, swing1LimitDegrees=45, swing2LimitDegrees=45, twistLimitDegrees=45}
 {bone1Name=Dyn_leaf10, bone2Name=Dyn_Master_HIve, swing1LimitDegrees=30, swing2LimitDegrees=30, twistLimitDegrees=45}
 {bone1Name=Dyn_leaf11, bone2Name=Dyn_Master_HIve, swing1LimitDegrees=45, swing2LimitDegrees=45, twistLimitDegrees=45}
 {bone1Name=Dyn_leaf12, bone2Name=Dyn_Master_HIve, swing1LimitDegrees=45, swing2LimitDegrees=45, twistLimitDegrees=45}
 {bone1Name=Dyn_leaf13, bone2Name=Dyn_Master_HIve, swing1LimitDegrees=20, swing2LimitDegrees=20, twistLimitDegrees=45}
]
